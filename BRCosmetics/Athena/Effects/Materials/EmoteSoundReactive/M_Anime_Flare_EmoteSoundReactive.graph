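# Translucent unlit particle and Niagara sprite material for an anime-style flare that reacts to emote audio.
# Particle color times the flare texture's green channel drives HDR-remapped emissive.
# Opacity combines particle alpha, that channel, a quality-switched near-camera fade and blended low/high audio levels from FortniteMaterialParameters.
# A low-audio threshold gates opacity and, when silent, produces an offset toward the object position scaled by ParticleScale, collapsing the sprite.
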Material /BRCosmetics/Athena/Effects/Materials/EmoteSoundReactive/M_Anime_Flare_EmoteSoundReactive
graph {
 properties {
  bEnableResponsiveAA=true
  bUsedWithNiagaraSprites=true
  bUsedWithParticleSprites=true
  blendMode=BLEND_Translucent
  pixelDepthOffsetMode=PDOM_Legacy
  shadingModel=MSM_Unlit
  usedShadingModels=Unlit
 }
 nodes [
  n1 = ObjectPositionWS()
  n2 = WorldPosition()
  n3 = CollectionParameter() {collection=&FortniteMaterialParameters, parameterName=Emote_AudioLow}
  n4 = Constant() {r=1}
  n5 = If(A=@n3, B=Constant() {r=0.01}, "A > B"=@n4, "A == B"=@n4, "A < B"=Constant())
  LinearInterpolate(A=Multiply(A=Subtract(A=@n1, B=@n2.XYZ), B=ScalarParameter() {parameterName=ParticleScale}), B=Constant(), Alpha=@n5)
  n6 = ParticleColor()
  n7 = TextureSampleParameter2D() {parameterName=Flare_Texture, texture=/Game/Effects/Fort_Effects/Textures/Flares/T_Anime_Flare}
  n8 = Multiply(A=@n6.RGB, B=@n7.G)
  MaterialFunctionCall(Emissive=MaterialFunctionCall(LDR Input=Clamp(None=@n8), HDR Tint=@n8) {materialFunction=&DeriveHDRfromLDR}.RemappedRange) {materialFunction=&MF_FX_EmissiveAdjustments}
  n9 = Multiply(A=@n6.A, B=@n7.G)
  n10 = MaterialFunctionCall(Distance=ScalarParameter() {defaultValue=200, parameterName=CameraDistanceFade}) {materialFunction=&Near_Camera_Fade}
  n11 = QualitySwitch(Default=@n10.Result, Low=Constant() {r=1}, High=@n10.Result, Medium=@n10.Result)
  n12 = CollectionParameter() {collection=&FortniteMaterialParameters, parameterName=Emote_AudioLow}
  n13 = CollectionParameter() {collection=&FortniteMaterialParameters, parameterName=Emote_AudioHigh}
  Multiply(A=Multiply(A=Multiply(A=@n9, B=@n11), B=@n5), B=LinearInterpolate(A=@n12, B=@n13, Alpha=@n13))
 ]
}
