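# Additive, unlit, two-sided mesh-particle material for a hollow wave effect that reacts to emote audio.
# Squared view-facing falloff, vertex red and turbulence panned by dynamic-parameter Time form a mask blending a tint into particle color for emissive.
# The Emote_AudioLow collection parameter gates that mask, which is scaled by a low/high audio blend.
# Below a small threshold, the same gate enables an offset toward object position, scaled by ParticleScale.
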Material /BRCosmetics/Athena/Effects/Materials/EmoteSoundReactive/M_HollowWave_EmoteSoundReactive
graph {
 properties {
  bUsedWithMeshParticles=true
  blendMode=BLEND_Additive
  pixelDepthOffsetMode=PDOM_Legacy
  shadingModel=MSM_Unlit
  translucencyPass=MTP_BeforeDOF
  twoSided=true
  usedShadingModels=Unlit
 }
 nodes [
  n1 = ObjectPositionWS()
  n2 = WorldPosition()
  n3 = CollectionParameter() {collection=&FortniteMaterialParameters, parameterName=Emote_AudioLow}
  n4 = Constant() {r=1}
  n5 = If(A=@n3, B=Constant() {r=0.01}, "A > B"=@n4, "A == B"=@n4, "A < B"=Constant())
  LinearInterpolate(A=Multiply(A=Subtract(A=@n1, B=@n2.XYZ), B=ScalarParameter() {parameterName=ParticleScale}), B=Constant(), Alpha=@n5)
  n6 = ParticleColor()
  n7 = Abs(None=DotProduct(A=VertexNormalWS(), B=CameraVectorWS()))
  n8 = TextureCoordinate() {vTiling=0.1}
  n9 = DynamicParameter() {paramNames [Time,Random,Param3,Ref]}
  n10 = Add(A=@n9.Time, B=@n9.Time)
  n11 = Multiply(
   A=Multiply(A=VertexColor().R, B=Multiply(A=@n7, B=@n7))
   B=TextureSample(UVs=Panner(Coordinate=@n8, Time=@n10) {speedY=0.5}) {texture=/Game/Global/Textures/Default/Caustics/T_Turbulence_Seamless}.R
  )
  n12 = LinearInterpolate(A=VectorParameter() {defaultValue=<rgba 0.4948 0.4948 0.4948 1>}.RGB, B=@n6.RGB, Alpha=@n11)
  n13 = Clamp(None=@n12) {maxDefault=3}
  MaterialFunctionCall(Emissive=Multiply(A=ShadingPathSwitch(Default=@n12, Mobile=@n13) {desc=Migrated from FeatureLevelSwitch}, B=@n6.A)) {materialFunction=&MF_FX_EmissiveAdjustments}
  n14 = CollectionParameter() {collection=&FortniteMaterialParameters, parameterName=Emote_AudioLow}
  n15 = CollectionParameter() {collection=&FortniteMaterialParameters, parameterName=Emote_AudioHigh}
  Multiply(A=Multiply(A=@n5, B=@n11), B=LinearInterpolate(A=@n14, B=@n15, Alpha=@n15))
 ]
}
