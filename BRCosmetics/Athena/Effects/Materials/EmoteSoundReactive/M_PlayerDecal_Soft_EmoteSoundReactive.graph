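# Translucent unlit particle sprite material for a soft, decal-like player effect that reacts to emote audio.
# Scene depth and an ambient-occlusion-driven curve atlas lookup form a mask, shaped by a sphere mask and depth fades; mobile skips the scene-texture mask.
# Particle color feeds emissive, while Emote_AudioLow and Emote_AudioHigh scale the mask.
# A low-audio threshold zeroes that mask in silence and switches on a ParticleScale-driven offset toward the object position.
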Material /BRCosmetics/Athena/Effects/Materials/EmoteSoundReactive/M_PlayerDecal_Soft_EmoteSoundReactive
graph {
 properties {
  bUsedWithParticleSprites=true
  blendMode=BLEND_Translucent
  pixelDepthOffsetMode=PDOM_Legacy
  shadingModel=MSM_Unlit
  twoSided=true
  usedShadingModels=Unlit
 }
 nodes [
  n1 = ObjectPositionWS()
  n2 = WorldPosition()
  n3 = CollectionParameter() {collection=&FortniteMaterialParameters, parameterName=Emote_AudioLow}
  n4 = Constant() {r=1}
  n5 = If(A=@n3, B=Constant() {r=0.01}, "A > B"=@n4, "A == B"=@n4, "A < B"=Constant())
  LinearInterpolate(A=Multiply(A=Subtract(A=@n1, B=@n2.XYZ), B=ScalarParameter() {parameterName=ParticleScale}), B=Constant(), Alpha=@n5)
  n6 = Subtract(A=ParticlePositionWS(), B=CameraPositionWS())
  n7 = Add(
   A=Abs(None=MaterialFunctionCall(Vector 3=@n6) {materialFunction=/Engine/Functions/Engine_MaterialFunctions02/Utility/VectorLength}.V3 Length)
   B=ScalarParameter() {defaultValue=25, group=DepthMask, parameterName=DethLength}
  ) {constB=40}
  n8 = Clamp(None=Multiply(A=SceneTexture() {sceneTextureId=PPI_SceneDepth}.Color, B=Divide(B=@n7)) {constB=0.0043})
  n9 = CurveAtlasRowParameter(CurveTime=ComponentMask(None=SceneTexture() {sceneTextureId=PPI_AmbientOcclusion}.Color) {r=true}) {
   atlas=&AthenaCurveAtlas01
   curve=/Game/Athena/LinearColorCurves/Curve_JellyfishPickaxe_caustic
   defaultValue=99
   group=DepthMask
   sortPriority=1
  }
  n10 = Multiply(A=ComponentMask(None=OneMinus(None=@n8)) {r=true}, B=@n9.G)
  n11 = SphereMask(
   A=TextureCoordinate()
   B=Constant() {r=0.5}
   Radius=ScalarParameter() {defaultValue=0.4, parameterName=SphereRadius}
   Hardness=ScalarParameter() {defaultValue=-1, parameterName=SphereHardness}
  ) {attenuationRadius=0.5, hardnessPercent=0}
  n12 = ShadingPathSwitch(Default=DepthFade(Opacity=Multiply(A=@n10, B=@n11)) {fadeDistanceDefault=10}, Mobile=@n11) {desc=Migrated from FeatureLevelSwitch}
  n13 = TextureCoordinate()
  n14 = MaterialFunctionCall(Float2=Sine(None=Multiply(A=@n13) {constB=0.5})) {materialFunction=&BreakOutFloat2Components}
  n15 = Multiply(A=@n14.R, B=@n14.R)
  n16 = Constant() {r=300}
  n17 = Multiply(
   A=Multiply(
    A=Multiply(A=@n12, B=@n15)
    B=MaterialFunctionCall(Fade Length=@n16) {materialFunction=&CameraDepthFade}.Result
   )
   B=ParticleColor().A
  )
  MaterialFunctionCall(Emissive=StaticSwitchParameter(True=@n17, False=ParticleColor().RGB) {parameterName=TestView}) {materialFunction=&MF_FX_EmissiveAdjustments}
  n18 = Constant() {r=1}
  n19 = CollectionParameter() {collection=&FortniteMaterialParameters, parameterName=Emote_AudioLow}
  n20 = CollectionParameter() {collection=&FortniteMaterialParameters, parameterName=Emote_AudioHigh}
  Multiply(A=Multiply(A=StaticSwitchParameter(True=@n18, False=@n17) {parameterName=TestView}, B=@n5), B=LinearInterpolate(A=@n19, B=@n20, Alpha=@n20))
  ScalarParameter() {defaultValue=1, parameterName=Param_1}
  ShadingPathSwitch()
 ]
}
