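# Unlit, translucent, two-sided mesh-particle material for an emote shockwave ring that reacts to emote audio.
# Color comes from particle color, time-based hue shifting or a panned rainbow gradient remapped to HDR, selected by static switches.
# Opacity combines a rotated ring texture, particle alpha and dynamic alpha scale with low/high audio levels.
# A low-audio threshold gates opacity and an offset toward the object's position, scaled by ParticleScale, that applies during silence.
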Material /BRCosmetics/Athena/Effects/Materials/EmoteSoundReactive/M_Simple_Particle_Shockwave_EmoteSoundReactive
graph {
 properties {
  bUsedWithMeshParticles=true
  blendMode=BLEND_Translucent
  pixelDepthOffsetMode=PDOM_Legacy
  shadingModel=MSM_Unlit
  twoSided=true
  usedShadingModels=Unlit
 }
 nodes [
  n1 = ObjectPositionWS()
  n2 = WorldPosition()
  n3 = CollectionParameter() {collection=&FortniteMaterialParameters, parameterName=Emote_AudioLow}
  n4 = Constant() {r=1}
  n5 = If(A=@n3, B=Constant() {r=0.01}, "A > B"=@n4, "A == B"=@n4, "A < B"=Constant())
  LinearInterpolate(A=Multiply(A=Subtract(A=@n1, B=@n2.XYZ), B=ScalarParameter() {parameterName=ParticleScale}), B=Constant(), Alpha=@n5)
  n6 = ComponentMask(None=TextureCoordinate() {uTiling=0, vTiling=0}) {g=true}
  n7 = Multiply(
   A=TextureSample(UVs=Panner(Coordinate=@n6) {speedX=1.6667}) {texture=/Game/Global/Textures/Default/Masks_and_Gradients/T_Grad_Rainbow}.RGB
   B=ParticleColor().RGB
  )
  n8 = Clamp(None=@n7)
  n9 = FeatureLevelSwitch(Default=@n7, ES3_1=Clamp(None=@n7) {maxDefault=3})
  n10 = ScalarParameter() {defaultValue=1, parameterName=Speed}
  n11 = Time()
  n12 = ParticleColor()
  n13 = Multiply(
   A=MaterialFunctionCall(Hue Shift Percentage=Multiply(A=@n10, B=@n11), Texture=VectorParameter() {parameterName=StartColor, defaultValue=<rgba 1 0 0 1>}.RGB) {materialFunction=&HueShift}.Result
   B=@n12.RGB
  )
  MaterialFunctionCall(Emissive=StaticSwitchParameter(
    True=MaterialFunctionCall(LDR Input=@n8, HDR Tint=@n9) {materialFunction=&DeriveHDRfromLDR}.RemappedRange
    False=StaticSwitchParameter(True=@n13, False=@n12.RGB) {parameterName=UseRainbowColor}
   ) {parameterName=UseMellowRainbowColors}) {materialFunction=&MF_FX_EmissiveAdjustments}
  n14 = DynamicParameter() {paramNames [ScaleAlphaOnly, Param2, RotationInitial, RotationTime]}
  n15 = Multiply(A=@n14.ScaleAlphaOnly, B=@n12.A)
  n16 = TextureSample(UVs=Rotator(Time=Add(A=@n14.RotationInitial, B=@n14.RotationInitial))) {samplerSource=SSM_Clamp_WorldGroupSettings, texture=/Game/Effects/Fort_Effects/Textures/Outlander/T_InTheZone_GroundRing_01}
  n17 = CollectionParameter() {collection=&FortniteMaterialParameters, parameterName=Emote_AudioLow}
  n18 = CollectionParameter() {collection=&FortniteMaterialParameters, parameterName=Emote_AudioHigh}
  Multiply(A=Multiply(A=Multiply(A=@n15, B=@n16.G), B=@n5), B=LinearInterpolate(A=@n17, B=@n18, Alpha=@n18))
 ]
}
