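# Translucent unlit mesh-particle material for a sound-reactive emote wake effect with rippling, glowing water-like caustics.
# Dynamic-parameter-driven radial sine ripples distort noise and bubble textures, which depth fade and a sphere mask shape into particle-colored emissive.
# Ripples and caustic noise also build vertex displacement, with optional center push.
# A threshold on the Emote_AudioLow collection parameter gates depth-faded, optionally sphere-masked opacity and switches the displacement's scale term.
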Material /BRCosmetics/Athena/Effects/Materials/EmoteSoundReactive/M_Wake_EmoteSoundReactive
graph {
 properties {
  bEnableResponsiveAA=true
  bUsedWithMeshParticles=true
  blendMode=BLEND_Translucent
  pixelDepthOffsetMode=PDOM_Legacy
  shadingModel=MSM_Unlit
  translucencyLightingMode=TLM_VolumetricPerVertexNonDirectional
  usedShadingModels=Unlit
 }
 nodes [
  n1 = ParticleColor()
  n2 = TextureCoordinate()
  n3 = ScalarParameter() {defaultValue=6, parameterName=Caustic_Scale}
  n4 = Constant() {r=0.5}
  n5 = TextureCoordinate()
  n6 = DynamicParameter() {defaultValue=<rgba 0.2 1 1 1>, paramNames [Ripple, RippleScale, FadeDistance, ScaleCausticHeight]}
  n7 = Multiply(
   A=Power(Base=Distance(A=@n4, B=@n5)) {constExponent=1.4}
   B=LinearInterpolate(
    A=ScalarParameter() {defaultValue=3, parameterName=NumRipplesStart}
    B=ScalarParameter() {defaultValue=10, parameterName=NumRipplesEnd}
    Alpha=@n6.Ripple
   ) {constA=12, constB=3}
  ) {constB=2}
  n8 = Multiply(A=Multiply(A=@n6.Ripple) {constB=-1}, B=ScalarParameter() {defaultValue=3, parameterName=RippleSpeed}) {constB=-5}
  n9 = ConstantBiasScale(None=Sine(None=Add(A=@n7, B=@n8)))
  n10 = Multiply(A=ScalarParameter() {defaultValue=0.1, parameterName=UVDistort}, B=@n9)
  n11 = TextureSample(UVs=Add(A=Multiply(A=@n2, B=@n3), B=@n10)) {samplerType=SAMPLERTYPE_LinearColor, texture=/Game/Characters/Player/Male/Medium/Bodies/M_MED_SRDrift/Textures/T_ElectricNoise}
  n12 = TextureCoordinate()
  n13 = ScalarParameter() {defaultValue=5, parameterName=BubbleScale}
  n14 = TextureSample(UVs=Add(A=Multiply(A=@n12, B=@n13), B=@n10)) {texture=/Game/Effects/Fort_Effects/Textures/Water/T_MaskeSplashes_Bubbles1_A}
  n15 = DynamicParameter() {defaultValue=<rgba 0.2 1 1 1>, paramNames [Ripple, RippleScale, FadeDistance, ScaleCausticHeight]}
  n16 = LinearInterpolate(
   A=Add(A=@n11.G, B=@n14.RGB)
   B=Multiply(A=ScalarParameter() {defaultValue=2.25, parameterName=ScaleGlow}, B=@n14.RGB) {constA=2}
   Alpha=DepthFade(FadeDistance=Multiply(A=@n15.FadeDistance) {constB=0.8})
  )
  n17 = Power(
   Base=SphereMask(A=TextureCoordinate(), B=Constant() {r=0.5}, Radius=ScalarParameter() {defaultValue=0.35, parameterName=Radius}) {attenuationRadius=0.35, hardnessPercent=1}
   Exp=ScalarParameter() {defaultValue=2, parameterName=SphereMaskContrast}
  )
  MaterialFunctionCall(Emissive=Multiply(A=@n1.RGB, B=Multiply(A=@n16, B=@n17))) {materialFunction=&MF_FX_EmissiveAdjustments}
  n18 = Subtract(A=ObjectPositionWS(), B=WorldPosition().XYZ)
  n19 = ScalarParameter() {parameterName=ParticleScale}
  n20 = CollectionParameter() {collection=&FortniteMaterialParameters, parameterName=Emote_AudioLow}
  n21 = Constant() {r=1}
  n22 = If(A=@n20, B=Constant() {r=0.01}, "A > B"=@n21, "A == B"=@n21, "A < B"=Constant())
  n23 = Multiply(A=@n11.G, B=@n15.ScaleCausticHeight)
  n24 = VectorParameter() {parameterName=RippleHeight, defaultValue=<rgba 0 0 2 0>}
  n25 = DynamicParameter() {defaultValue=<rgba 0.2 1 1 1>, paramNames [Ripple, RippleScale, FadeDistance, ScaleCausticHeight]}
  n26 = Transform(None=Multiply(A=Add(A=@n23, B=@n9), B=Multiply(A=@n24.RGB, B=@n25.RippleScale))) {transformSourceType=TRANSFORMSOURCE_Local}
  n27 = TextureCoordinate()
  n28 = Constant() {r=0.5}
  n29 = ScalarParameter() {defaultValue=0.5, parameterName=WPO_CenterRadius}
  n30 = ScalarParameter() {parameterName=WPO_CenterHardness}
  n31 = Multiply(
   A=Multiply(A=SphereMask(A=@n27, B=@n28, Radius=@n29, Hardness=@n30), B=ScalarParameter() {parameterName=WPO_CenterPush})
   B=Constant3Vector() {constant=<rgba 0 0 1 0>}
  )
  Multiply(
   A=LinearInterpolate(A=Multiply(A=@n18, B=@n19), B=Constant(), Alpha=@n22)
   B=StaticSwitchParameter(True=Add(A=@n26, B=@n31), False=@n26) {parameterName=Use Centeral Push on Mesh}
  )
  n32 = Multiply(A=@n1.A, B=DepthFade(FadeDistance=@n15.FadeDistance))
  n33 = Power(
   Base=SphereMask(A=TextureCoordinate(), B=Constant() {r=0.5}, Radius=ScalarParameter() {defaultValue=0.35, parameterName=Alpha_Radius}) {attenuationRadius=0.35, hardnessPercent=1}
   Exp=ScalarParameter() {defaultValue=2, parameterName=Alpha_SphereMaskContrast}
  )
  Multiply(A=StaticSwitchParameter(True=Multiply(A=@n32, B=@n33), False=@n32) {parameterName=Use SphereMask Alpha}, B=@n22)
 ]
}
